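# Niagara system for the MP variant of the Daikon vehicle drift trail cosmetic, a snow-themed wheel effect.
# Rock_Smoke streams meshes and smoke, mist and spark sprites backward in six spawn groups, while Snowball001 and WheelWrap_Snow001 keep meshes attached to wheel bones.
# Partitioned particles pick bones by index, and FN_DisableTireFX masks them per wheel.
# Vehicle speed and a front-end preview flag drive Rock_Smoke's rates and velocities.
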NiagaraSystem /VehicleCosmetics/DriftTrail/DriftTrail_Daikon/FX/NS_DriftTrail_Daikon_MP
dependencies {
 usedDataInterfaces [
  &NiagaraDataInterfaceCurve
  &NiagaraDataInterfaceVectorCurve
  &NiagaraDataInterfaceSpriteRendererInfo
  &NiagaraDataInterfaceVector2DCurve
  &NiagaraDataInterfaceColorCurve
  &NiagaraDataInterfaceSkeletalMesh
 ]
 usedDynamicInputs []
 usedModules [
  &self:SystemSpawnScript.SystemScriptSource.SystemScriptGraph.NiagaraNodeAssignment_0.SetVariables_#1
  &FN_UnpackVehicleFlags
  /VehicleCosmetics/FX/Niagara/Modules/VC_SetVehicleSpeedAttribs
  &SystemState
  /Game/Effects/Niagara/Modules/Vehicle/FN_VehicleFlags_Cheatcodes
 ]
}
emitters {
 Rock_Smoke {
  emitterUpdateScript [
   &EmitterState()
   &SpawnRate(
    SpawnRate=/VehicleCosmetics/FX/Niagara/Modules/VC_Mask_Floats(
     FrontEnd Value:NiagaraFloat=&RandomRangeFloat(Minimum=25, Maximum=30)
     blsFrontEndPreview:NiagaraBool=$User.bIsFrontendPreview
     SpeedMin Value:NiagaraFloat=&RandomRangeFloat(Minimum=25, Maximum=30)
     SpeedMax Value:NiagaraFloat=&RandomRangeFloat(Minimum=25, Maximum=30)
     Speed Alpha:NiagaraFloat=$System.Speed_Alpha
    )
    Use Spawn Probability=false
   )
   &SpawnRate(
    SpawnGroup=1
    SpawnRate=&VC_Mask_Floats(
     FrontEnd Value:NiagaraFloat=&RandomRangeFloat(Minimum=150, Maximum=100)
     blsFrontEndPreview:NiagaraBool=$User.bIsFrontendPreview
     SpeedMin Value:NiagaraFloat=&RandomRangeFloat(Minimum=150, Maximum=100)
     SpeedMax Value:NiagaraFloat=&RandomRangeFloat(Minimum=150, Maximum=100)
     Speed Alpha:NiagaraFloat=$System.Speed_Alpha
    )
    Use Spawn Probability=false
   ) {module=SpawnRate001}
   &SpawnRate(
    SpawnGroup=2
    SpawnRate=&VC_Mask_Floats(
     FrontEnd Value:NiagaraFloat=&RandomRangeFloat(Minimum=120, Maximum=150)
     blsFrontEndPreview:NiagaraBool=$User.bIsFrontendPreview
     SpeedMin Value:NiagaraFloat=&RandomRangeFloat(Minimum=120, Maximum=150)
     SpeedMax Value:NiagaraFloat=&RandomRangeFloat(Minimum=120, Maximum=150)
     Speed Alpha:NiagaraFloat=$System.Speed_Alpha
    )
    Use Spawn Probability=false
   ) {module=SpawnRate002}
   &SpawnRate(
    SpawnGroup=3
    SpawnRate=&VC_Mask_Floats(
     FrontEnd Value:NiagaraFloat=&RandomRangeFloat(Minimum=10, Maximum=30)
     blsFrontEndPreview:NiagaraBool=$User.bIsFrontendPreview
     SpeedMin Value:NiagaraFloat=&RandomRangeFloat(Minimum=10, Maximum=30)
     SpeedMax Value:NiagaraFloat=&RandomRangeFloat(Minimum=10, Maximum=30)
     Speed Alpha:NiagaraFloat=$System.Speed_Alpha
    )
    Use Spawn Probability=false
   ) {module=SpawnRate003}
   &SpawnRate(
    SpawnGroup=4
    SpawnRate=&VC_Mask_Floats(
     FrontEnd Value:NiagaraFloat=&RandomRangeFloat(Minimum=200, Maximum=160)
     blsFrontEndPreview:NiagaraBool=$User.bIsFrontendPreview
     SpeedMin Value:NiagaraFloat=&RandomRangeFloat(Minimum=80, Maximum=160)
     SpeedMax Value:NiagaraFloat=&RandomRangeFloat(Minimum=80, Maximum=160)
     Speed Alpha:NiagaraFloat=$System.Speed_Alpha
    )
    Use Spawn Probability=false
   ) {module=SpawnRate004}
   &SpawnRate(
    SpawnGroup=5
    SpawnRate=&VC_Mask_Floats(
     FrontEnd Value:NiagaraFloat=&RandomRangeFloat(Minimum=80, Maximum=160)
     blsFrontEndPreview:NiagaraBool=$User.bIsFrontendPreview
     SpeedMin Value:NiagaraFloat=&RandomRangeFloat(Minimum=80, Maximum=160)
     SpeedMax Value:NiagaraFloat=&RandomRangeFloat(Minimum=80, Maximum=160)
     Speed Alpha:NiagaraFloat=$System.Speed_Alpha
    )
    Use Spawn Probability=false
   ) {module=SpawnRate005}
  ]
  particleSpawnScript [
   &VC_PaintedVariantsOverride() {enabled=false}
   /SetVariables(Particles.TagsRockSmoke:NiagaraInt32=&SpawnGroup()) {isSetParameters=true}
   &InitializeParticle(
    Color Mode=::NewEnumerator0
    Lifetime Max=0.65
    Lifetime Min=0.45
    Lifetime Mode=::NewEnumerator1
    Mesh Scale Mode=::NewEnumerator1
    Mesh Uniform Scale Max=0.08
    Mesh Uniform Scale Min=0.04
    Sprite Rotation Mode=::NewEnumerator2
    Sprite Size Mode=::NewEnumerator1
    Uniform Sprite Size Max=40
    Uniform Sprite Size Min=20
    UsePositionOffset=false
    Write Material Random=true
   )
   &PartitionParticles()
   &SkeletalMeshLocation(
    Bone / Socket Index=&Add_Integer(A=$Particles.PartitionParticles.Partition, B=4)
    Bone / Socket Sampling Mode=::NewEnumerator5
    Interpolate Mesh Sampling=false
    MaskSpawnGroup=false
    Mesh Sampling Type=::NewEnumerator6
    Sampled Position Offset=&AddVectorToPosition(Position=<-70 0 -25>, Vector=$User.offset)
    Skeletal Mesh=$System.SkeletalMesh
   )
   &AddVelocity(Velocity=&VC_Mask_Vectors(
     FrontEnd Value:Vector3f=&MaskVectorBySpawnGroup(
      Masked Vector=&RandomRangeVector(Minimum=<-600 -30 300>, Maximum=<-350 30 100>)
      Passthrough Vector=<0 0 0>
     )
     blsFrontEndPreview:NiagaraBool=$User.bIsFrontendPreview
     SpeedMin Value:Vector3f=&MaskVectorBySpawnGroup(
      Masked Vector=&RandomRangeVector(Minimum=<-1500 -200 800>, Maximum=<-1200 200 -20>)
      Passthrough Vector=<0 0 0>
     )
     SpeedMax Value:Vector3f=&MaskVectorBySpawnGroup(
      Masked Vector=&RandomRangeVector(Minimum=<-1500 -200 800>, Maximum=<-1200 200 -20>)
      Passthrough Vector=<0 0 0>
     )
     Speed Alpha:NiagaraFloat=$System.Speed_Alpha
    ))
   &AddVelocity(Velocity=&VC_Mask_Vectors(
     FrontEnd Value:Vector3f=&MaskVectorBySpawnGroup(
      Masked Vector=&RandomRangeVector(Minimum=<-1500 10 0>, Maximum=<-1000 -10 0>)
      Passthrough Vector=<0 0 0>
      Spawn Group Mask=1
     )
     blsFrontEndPreview:NiagaraBool=$User.bIsFrontendPreview
     SpeedMin Value:Vector3f=&MaskVectorBySpawnGroup(
      Masked Vector=&RandomRangeVector(Minimum=<-1500 10 0>, Maximum=<-1000 -10 0>)
      Passthrough Vector=<0 0 0>
      Spawn Group Mask=1
     )
     SpeedMax Value:Vector3f=&MaskVectorBySpawnGroup(
      Masked Vector=&RandomRangeVector(Minimum=<-1500 10 0>, Maximum=<-1000 -10 0>)
      Passthrough Vector=<0 0 0>
      Spawn Group Mask=1
     )
     Speed Alpha:NiagaraFloat=$System.Speed_Alpha
    )) {module=AddVelocity001}
   &AddVelocity(Velocity=&VC_Mask_Vectors(
     FrontEnd Value:Vector3f=&MaskVectorBySpawnGroup(
      Masked Vector=&RandomRangeVector(Minimum=<-1500 10 0>, Maximum=<-1000 -10 0>)
      Passthrough Vector=<0 0 0>
      Spawn Group Mask=2
     )
     blsFrontEndPreview:NiagaraBool=$User.bIsFrontendPreview
     SpeedMin Value:Vector3f=&MaskVectorBySpawnGroup(
      Masked Vector=&RandomRangeVector(Minimum=<-1500 -10 0>, Maximum=<-1000 10 0>)
      Passthrough Vector=<0 0 0>
      Spawn Group Mask=2
     )
     SpeedMax Value:Vector3f=&MaskVectorBySpawnGroup(
      Masked Vector=&RandomRangeVector(Minimum=<-1500 -10 0>, Maximum=<-1000 10 0>)
      Passthrough Vector=<0 0 0>
      Spawn Group Mask=2
     )
     Speed Alpha:NiagaraFloat=$System.Speed_Alpha
    )) {module=AddVelocity002}
   &AddVelocity(Velocity=&VC_Mask_Vectors(
     FrontEnd Value:Vector3f=&MaskVectorBySpawnGroup(
      Masked Vector=&RandomRangeVector(Minimum=<-1500 -60 60>, Maximum=<-1200 60 -60>)
      Passthrough Vector=<0 0 0>
      Spawn Group Mask=3
     )
     blsFrontEndPreview:NiagaraBool=$User.bIsFrontendPreview
     SpeedMin Value:Vector3f=&MaskVectorBySpawnGroup(
      Masked Vector=&RandomRangeVector(Minimum=<-1500 -60 60>, Maximum=<-1200 60 -60>)
      Passthrough Vector=<0 0 0>
      Spawn Group Mask=3
     )
     SpeedMax Value:Vector3f=&MaskVectorBySpawnGroup(
      Masked Vector=&RandomRangeVector(Minimum=<-1500 -60 60>, Maximum=<-1200 60 -60>)
      Passthrough Vector=<0 0 0>
      Spawn Group Mask=3
     )
     Speed Alpha:NiagaraFloat=$System.Speed_Alpha
    )) {module=AddVelocity003}
   &AddVelocity(Velocity=&VC_Mask_Vectors(
     FrontEnd Value:Vector3f=&MaskVectorBySpawnGroup(
      Masked Vector=&RandomRangeVector(Minimum=<-600 0 0>, Maximum=<-350 0 0>)
      Passthrough Vector=<0 0 0>
      Spawn Group Mask=4
     )
     blsFrontEndPreview:NiagaraBool=$User.bIsFrontendPreview
     SpeedMin Value:Vector3f=&MaskVectorBySpawnGroup(
      Masked Vector=&RandomRangeVector(Minimum=<-1500 0 0>, Maximum=<-1200 0 0>)
      Passthrough Vector=<0 0 0>
      Spawn Group Mask=4
     )
     SpeedMax Value:Vector3f=&MaskVectorBySpawnGroup(
      Masked Vector=&RandomRangeVector(Minimum=<-1500 0 0>, Maximum=<-1200 0 0>)
      Passthrough Vector=<0 0 0>
      Spawn Group Mask=4
     )
     Speed Alpha:NiagaraFloat=$System.Speed_Alpha
    )) {module=AddVelocity004}
   &AddVelocity(Velocity=&VC_Mask_Vectors(
     FrontEnd Value:Vector3f=&MaskVectorBySpawnGroup(
      Masked Vector=&RandomRangeVector(Minimum=<-600 -100 100>, Maximum=<-350 100 -20>)
      Passthrough Vector=<0 0 0>
      Spawn Group Mask=5
     )
     blsFrontEndPreview:NiagaraBool=$User.bIsFrontendPreview
     SpeedMin Value:Vector3f=&MaskVectorBySpawnGroup(
      Masked Vector=&RandomRangeVector(Minimum=<-1500 -100 100>, Maximum=<-1200 100 -20>)
      Passthrough Vector=<0 0 0>
      Spawn Group Mask=5
     )
     SpeedMax Value:Vector3f=&MaskVectorBySpawnGroup(
      Masked Vector=&RandomRangeVector(Minimum=<-1500 -100 100>, Maximum=<-1200 100 -20>)
      Passthrough Vector=<0 0 0>
      Spawn Group Mask=5
     )
     Speed Alpha:NiagaraFloat=$System.Speed_Alpha
    )) {module=AddVelocity005}
   &ShapeLocation(MaskSpawnGroup=true, Override Seed=false, Sphere Radius=10, UseOwnerScale=true)
   &ShapeLocation(MaskSpawnGroup=true, Override Seed=false, Spawn Group Mask=1, Sphere Radius=1, Sphere Surface Distribution=1, UseOwnerScale=true) {module=ShapeLocation001}
   &ShapeLocation(MaskSpawnGroup=true, Override Seed=false, Spawn Group Mask=2, Sphere Radius=1, Sphere Surface Distribution=1, UseOwnerScale=true) {module=ShapeLocation002}
   &ShapeLocation(
    MaskSpawnGroup=true
    Override Seed=false
    Spawn Group Mask=3
    Sphere Radius=&RandomRangeFloat(Minimum=10, Maximum=20)
    Sphere Surface Distribution=1
    UseOwnerScale=true
   ) {module=ShapeLocation003}
   &ShapeLocation(MaskSpawnGroup=true, Override Seed=false, Spawn Group Mask=4, Sphere Radius=1, Sphere Surface Distribution=1, UseOwnerScale=true) {module=ShapeLocation004}
   &ShapeLocation(MaskSpawnGroup=true, Override Seed=false, Spawn Group Mask=5, Sphere Radius=1, Sphere Surface Distribution=1, UseOwnerScale=true) {module=ShapeLocation005}
   &InitialMeshOrientation(Use Rotation Vector=false)
   /SetVariables(
    Particles.ColorMist:LinearColor=<rgba 0.5 0.8 1 0.1>
    Particles.ColorSmoke:LinearColor=<rgba 7 41.2 64 0.122>
    Particles.SpriteSizeSnowMist:Vector2f=&RandomRangeVector2D(Minimum=<0 0>, Maximum=<0 0>)
    Particles.SpriteSizeSparks:Vector2f=&RandomRangeVector2D(
     Minimum=&Multiply_Vector2D_ByFloat(Float=30, Vector2D=<0.1 0.5>)
     Maximum=&Multiply_Vector2D_ByFloat(Float=50, Vector2D=<0.01 0.1>)
    )
    Particles.SubImageIndex_1:NiagaraFloat=$StackContext.SubImageIndex
   ) {isSetParameters=true}
  ]
  particleUpdateScript [
   &FN_DisableTireFX(
    Mask Index Attribute:NiagaraInt32=$Particles.PartitionParticles.Partition
    Mask Index FL=2
    Mask Index FR=3
    Mask Index RL=0
    Mask Index RR=1
   )
   &ParticleState()
   &InheritSourceMovement(
    Applied Position Scale=&Multiply_VectorByFloat(
     Vector=<1 1 1>
     Float=&FloatFromCurve(FloatCurve=DataInterface() {propertyValues {Curve {keys [
          {interpMode=RCIM_Cubic, leaveTangent=-0.0051, tangentMode=RCTM_Break, time=0, value=1}
          {arriveTangent=-0.247, interpMode=RCIM_Cubic, tangentMode=RCTM_Break, time=1, value=0.9}
         ]}}})
    )
    ApplyForce=false
    ApplyPosition=true
    ApplyPositionFalloff=false
   )
   &ScaleSpriteSize(
    Uniform Curve Scale=2.5
    Uniform Curve Sprite Scale=DataInterface() {propertyValues {Curve {keys [{tangentMode=RCTM_User, time=0, value=0.55}, {tangentMode=RCTM_User, time=1, value=1}]}}}
   )
   &ScaleMeshSize(Scale Factor=&MakeCustomVectorFromBool(
     Bool=$User.bIsFrontendPreview
     True Vector=&Multiply_VectorByFloat(
      Vector=<1 1 1>
      Float=&MakeFloatFromVector(
       Vector=&Multiply_VectorByFloat(
        Vector=<1 1 1>
        Float=&FloatFromCurve(FloatCurve=DataInterface() {propertyValues {Curve {keys [
             {interpMode=RCIM_Cubic, tangentMode=RCTM_Break, time=0, value=1}
             {arriveTangent=-3.1668, interpMode=RCIM_Cubic, tangentMode=RCTM_Break, time=1, value=0}
            ]}}})
       )
       Channel=::NewEnumerator0
      )
     )
     False Vector=&Multiply_VectorByFloat(
      Vector=<1 1 1>
      Float=&MakeFloatFromVector(
       Vector=&Multiply_VectorByFloat(
        Vector=<1 1 1>
        Float=&FloatFromCurve(FloatCurve=DataInterface() {propertyValues {Curve {keys [
             {interpMode=RCIM_Cubic, tangentMode=RCTM_Break, time=0, value=1}
             {arriveTangent=-5.8041, interpMode=RCIM_Cubic, tangentMode=RCTM_Break, time=0.4, value=0}
            ]}}})
       )
       Channel=::NewEnumerator0
      )
     )
    ))
   &GravityForce(Gravity=&MaskVectorBySpawnGroup(
     Masked Vector=&VectorFromCurve(
      VectorCurve=DataInterface() {propertyValues {
        XCurve {keys [<at 0 0>,<at 1 0>]}
        YCurve {keys [<at 0 0>,<at 1 0>]}
        ZCurve {keys [{interpMode=RCIM_Cubic, leaveTangent=1.1694, tangentMode=RCTM_Break, time=0, value=0.3}, <at 1 1>]}
       }}
      Scale Curve=<0 0 -980>
     )
     Passthrough Vector=<1 1 1>
    ))
   &SolveForcesAndVelocity(Clamp Velocity=false, Limit Acceleration=false)
   &DynamicMaterialParameters(
    Index 0 Param 1=0.35
    Index 0 Param 2=&RandomRangeFloat(Minimum=0.1, Maximum=0.5)
    Index 0 Param 3=1
    Index 0 Param 4=&RandomRangeFloat(Minimum=-0.3, Maximum=-0.05)
    Module.Index0Param0Write=true
    Module.Index0Param1Write=true
    Module.Index0Param2Write=true
    Module.Index0Param3Write=true
   )
   /SetVariables(
    Particles.DynamicMaterialParameters_Mist:Vector4f=&MakeVector4(
     X=&FloatFromCurve(FloatCurve=DataInterface() {propertyValues {Curve {keys [
          {interpMode=RCIM_Cubic, tangentMode=RCTM_Break, time=0, value=1}
          {interpMode=RCIM_Cubic, tangentMode=RCTM_Break, time=0.5, value=1.5}
          {interpMode=RCIM_Cubic, tangentMode=RCTM_Break, time=1, value=0}
         ]}}})
     Y=&FloatFromCurve(FloatCurve=DataInterface() {propertyValues {Curve {keys [
          {interpMode=RCIM_Cubic, tangentMode=RCTM_Break, time=0, value=1}
          {interpMode=RCIM_Cubic, tangentMode=RCTM_Break, time=0.5, value=0.3}
          {interpMode=RCIM_Cubic, tangentMode=RCTM_Break, time=1, value=0}
         ]}}})
     Z=&FloatFromCurve(FloatCurve=DataInterface() {propertyValues {Curve {keys [
          {interpMode=RCIM_Cubic, tangentMode=RCTM_Break, time=0, value=0}
          {interpMode=RCIM_Cubic, tangentMode=RCTM_Break, time=0.8, value=0.5}
          {interpMode=RCIM_Cubic, tangentMode=RCTM_Break, time=1, value=1}
         ]}}})
     W=&FloatFromCurve(FloatCurve=DataInterface() {propertyValues {Curve {keys [
          {interpMode=RCIM_Cubic, tangentMode=RCTM_Break, time=0, value=4}
          {interpMode=RCIM_Cubic, tangentMode=RCTM_Break, time=0.4, value=3}
          {interpMode=RCIM_Cubic, tangentMode=RCTM_Break, time=1, value=8}
         ]}}})
    )
    Particles.DynamicMaterialParameters_SmokeCone:Vector4f=&MakeVector4(
     X=&RandomRangeFloat(Minimum=0.3, Maximum=0.35)
     Y=&RandomRangeFloat(Minimum=0.5, Maximum=0.1)
     Z=0.4
     W=&RandomRangeFloat(Minimum=-0.3, Maximum=-0.4)
    )
   ) {isSetParameters=true}
   &SubUVAnimation(
    End Frame Range Override=64
    Random Start Frame=true
    "Sprite Renderer "=DataInterface() {propertyValues {SpriteRenderer=&self:Rock_Smoke_1.NiagaraSpriteRendererProperties_5}}
    Start Frame Offset:NiagaraInt32=&RandomRangeInt(Minimum=32, Maximum=64)
    Start Frame Range Override=32
    UseEndFrame=true
    UseStartFrame=true
   )
   /SetVariables(
    Particles.ColorMist:LinearColor=&MakeCustomLinearColorFromBool(
     Bool=$User.bIsFrontendPreview
     True Color=&ColorFromCurve(ColorCurve=DataInterface() {propertyValues {
        AlphaCurve {keys [
          {interpMode=RCIM_Cubic, tangentMode=RCTM_Break, time=0, value=0.225}
          {arriveTangent=-0.3957, interpMode=RCIM_Cubic, leaveTangent=-0.4437, tangentMode=RCTM_Break, time=0.6, value=0.15}
          {interpMode=RCIM_Cubic, leaveTangent=-0.2, time=1, value=0}
         ]}
        BlueCurve {keys [<at 0 0.4627 RCIM_Cubic>,<at 1 0.4627 RCIM_Cubic>]}
        GreenCurve {keys [<at 0 0.2745 RCIM_Cubic>,<at 1 0.2745 RCIM_Cubic>]}
        RedCurve {keys [<at 0 0.102 RCIM_Cubic>,<at 1 0.102 RCIM_Cubic>]}
       }})
     False Color=&ColorFromCurve(ColorCurve=DataInterface() {propertyValues {
        AlphaCurve {keys [
          <at 0 0.225>
          <at 0.3892 0.1764>
          {arriveTangent=-0.3957, interpMode=RCIM_Cubic, leaveTangent=-0.4437, tangentMode=RCTM_Break, time=0.6, value=0.15}
          {interpMode=RCIM_Cubic, leaveTangent=-0.2, time=1, value=0}
         ]}
        BlueCurve {keys [<at 0 0.4627 RCIM_Cubic>,<at 1 0.4627 RCIM_Cubic>]}
        GreenCurve {keys [<at 0 0.2745 RCIM_Cubic>,<at 1 0.2745 RCIM_Cubic>]}
        RedCurve {keys [<at 0 0.102 RCIM_Cubic>,<at 1 0.102 RCIM_Cubic>]}
       }})
    )
    Particles.ColorMistCone:LinearColor=&MakeCustomLinearColorFromBool(
     Bool=$User.bIsFrontendPreview
     True Color=&ColorFromCurve(ColorCurve=DataInterface() {propertyValues {
        AlphaCurve {keys [
          {interpMode=RCIM_Cubic, leaveTangent=-0.0486, tangentMode=RCTM_Break, time=0, value=0.2}
          {arriveTangent=-0.2022, interpMode=RCIM_Cubic, leaveTangent=-0.238, tangentMode=RCTM_Break, time=0.6, value=0.1125}
          {arriveTangent=-0.3, interpMode=RCIM_Cubic, leaveTangent=-0.5793, tangentMode=RCTM_Break, time=1, value=0}
         ]}
        BlueCurve {keys [<at 0 0.4902 RCIM_Cubic>,<at 1 0.4902 RCIM_Cubic>]}
        GreenCurve {keys [<at 0 0.2745 RCIM_Cubic>,<at 1 0.2745 RCIM_Cubic>]}
        RedCurve {keys [<at 0 0.0941 RCIM_Cubic>,<at 1 0.0941 RCIM_Cubic>]}
       }})
     False Color=&ColorFromCurve(ColorCurve=DataInterface() {propertyValues {
        AlphaCurve {keys [
          {interpMode=RCIM_Cubic, tangentMode=RCTM_Break, time=0, value=0.2}
          {arriveTangent=-0.2022, interpMode=RCIM_Cubic, leaveTangent=-0.403, tangentMode=RCTM_Break, time=0.6, value=0.1}
          {arriveTangent=-0.3704, interpMode=RCIM_Cubic, leaveTangent=-0.3704, tangentMode=RCTM_User, time=1, value=0}
         ]}
        BlueCurve {keys [<at 0 0.4902 RCIM_Cubic>,<at 1 0.4902 RCIM_Cubic>]}
        GreenCurve {keys [<at 0 0.2745 RCIM_Cubic>,<at 1 0.2745 RCIM_Cubic>]}
        RedCurve {keys [<at 0 0.0941 RCIM_Cubic>,<at 1 0.0941 RCIM_Cubic>]}
       }})
    )
    Particles.ColorSmoke:LinearColor=&ColorFromCurve(ColorCurve=DataInterface() {propertyValues {
       AlphaCurve {keys [<at 0 0.062 RCIM_Cubic>, {arriveTangent=-0.0032, interpMode=RCIM_Cubic, leaveTangent=-0.0032, tangentMode=RCTM_User, time=1, value=0}]}
       BlueCurve {keys [<at 0 64 RCIM_Cubic>,<at 1 64 RCIM_Cubic>]}
       GreenCurve {keys [<at 0 53.5521 RCIM_Cubic>,<at 1 53.5521 RCIM_Cubic>]}
       RedCurve {keys [<at 0 35.897 RCIM_Cubic>,<at 1 35.897 RCIM_Cubic>]}
      }})
    Particles.ColorSparks:LinearColor=&MakeCustomLinearColorFromBool(
     Bool=$User.bIsFrontendPreview
     True Color=&ColorFromCurve(ColorCurve=DataInterface() {propertyValues {
        AlphaCurve {keys [<at 0 1 RCIM_Cubic>,<at 1 0 RCIM_Cubic>]}
        BlueCurve {keys [<at 0 3 RCIM_Cubic>,<at 1 3 RCIM_Cubic>]}
        GreenCurve {keys [<at 0 0.6 RCIM_Cubic>,<at 1 0.6 RCIM_Cubic>]}
        RedCurve {keys [<at 0 0.4 RCIM_Cubic>,<at 1 0.4 RCIM_Cubic>]}
       }})
     False Color=&ColorFromCurve(ColorCurve=DataInterface() {propertyValues {
        AlphaCurve {keys [<at 0 0.1 RCIM_Cubic>,<at 1 0 RCIM_Cubic>]}
        BlueCurve {keys [<at 0 3 RCIM_Cubic>,<at 1 3 RCIM_Cubic>]}
        GreenCurve {keys [<at 0 2.6996 RCIM_Cubic>,<at 1 2.9104 RCIM_Cubic>]}
        RedCurve {keys [<at 0 1.4372 RCIM_Cubic>,<at 1 1.6218 RCIM_Cubic>]}
       }})
    )
    Particles.SpriteSizeMistCone:Vector2f=&Vector2DFromCurve(
     Vector2Curve=DataInterface() {propertyValues {
       XCurve {keys [{interpMode=RCIM_Cubic, tangentMode=RCTM_Break, time=0, value=1}, {interpMode=RCIM_Cubic, tangentMode=RCTM_Break, time=0.4, value=4}]}
       YCurve {keys [
         {interpMode=RCIM_Cubic, leaveTangentWeight=0.3333, tangentMode=RCTM_Break, time=0, value=1}
         {arriveTangentWeight=0.2167, interpMode=RCIM_Cubic, leaveTangentWeight=0.1167, tangentMode=RCTM_Break, time=0.4, value=4}
        ]}
      }}
     ScaleCurve=&RandomRangeVector2D(Minimum=<20 20>, Maximum=<35 35>)
    )
    Particles.SpriteSizeSmokeCone:Vector2f=&RandomRangeVector2D(
     Minimum=&Multiply_Vector2D_ByFloat(
      Float=&FloatFromCurve(FloatCurve=DataInterface() {propertyValues {Curve {keys [<at 0 0.55>,<at 1 1>]}}}, Scale Curve=4)
      Vector2D=<15 15>
     )
     Maximum=&Multiply_Vector2D_ByFloat(
      Float=&FloatFromCurve(FloatCurve=DataInterface() {propertyValues {Curve {keys [<at 0 0.55>,<at 1 1>]}}}, Scale Curve=4)
      Vector2D=<30 30>
     )
    )
    Particles.SpriteSizeSnowMist:Vector2f=&Vector2DFromCurve(
     Vector2Curve=DataInterface() {propertyValues {
       XCurve {keys [
         {interpMode=RCIM_Cubic, tangentMode=RCTM_Break, time=0, value=1}
         {interpMode=RCIM_Cubic, tangentMode=RCTM_Break, time=0.3, value=2.5}
         {interpMode=RCIM_Cubic, tangentMode=RCTM_Break, time=0.65, value=4}
        ]}
       YCurve {keys [
         {interpMode=RCIM_Cubic, leaveTangentWeight=0.3333, tangentMode=RCTM_Break, time=0, value=1}
         {arriveTangentWeight=0.1, interpMode=RCIM_Cubic, leaveTangentWeight=0.1167, tangentMode=RCTM_Break, time=0.3, value=2.5}
         {arriveTangentWeight=0.2167, interpMode=RCIM_Cubic, leaveTangentWeight=0.1167, tangentMode=RCTM_Break, time=0.65, value=4}
        ]}
      }}
     ScaleCurve=&RandomRangeVector2D(Minimum=<20 20>, Maximum=<25 25>)
    )
    Particles.colorsmoke2:LinearColor=&ColorFromCurve(ColorCurve=DataInterface() {propertyValues {
       AlphaCurve {keys [<at 0 0.062 RCIM_Cubic>, {arriveTangent=-0.0032, interpMode=RCIM_Cubic, leaveTangent=-0.0032, tangentMode=RCTM_User, time=1, value=0}]}
       BlueCurve {keys [<at 0 64 RCIM_Cubic>,<at 1 64 RCIM_Cubic>]}
       GreenCurve {keys [<at 0 41.2 RCIM_Cubic>,<at 1 60.7944 RCIM_Cubic>]}
       RedCurve {keys [<at 0 42 RCIM_Cubic>,<at 1 47.6565 RCIM_Cubic>]}
      }})
   ) {isSetParameters=true}
  ]
  rendererClasses [&NiagaraMeshRendererProperties, &NiagaraSpriteRendererProperties]
  renderers [
   {rendererClass=&NiagaraMeshRendererProperties, rendererIndex=0}
   {rendererClass=&NiagaraSpriteRendererProperties, rendererIndex=1}
   {rendererClass=&NiagaraSpriteRendererProperties, rendererIndex=2}
   {rendererClass=&NiagaraSpriteRendererProperties, rendererIndex=3}
   {rendererClass=&NiagaraSpriteRendererProperties, rendererIndex=4}
   {rendererClass=&NiagaraSpriteRendererProperties, rendererIndex=5}
  ]
 }
 Snowball001 {
  emitterUpdateScript [
   &EmitterState()
   &SpawnBurst_Instantaneous(Use Spawn Probability=false, UseLoopCountLimit=false) {module=SpawnBurst_Instantaneous001}
   &SpawnBurst_Instantaneous(Use Spawn Probability=false, UseLoopCountLimit=false)
   &SpawnRate(
    SpawnGroup=1
    SpawnRate=&VC_Mask_Floats(
     FrontEnd Value:NiagaraFloat=&RandomRangeFloat(Minimum=50, Maximum=70)
     blsFrontEndPreview:NiagaraBool=$User.bIsFrontendPreview
     SpeedMin Value:NiagaraFloat=&RandomRangeFloat(Minimum=0, Maximum=1)
     SpeedMax Value:NiagaraFloat=&RandomRangeFloat(Minimum=0, Maximum=1)
     Speed Alpha:NiagaraFloat=$System.Speed_Alpha
    )
    Use Spawn Probability=false
   ) {enabled=false}
   /SetVariables(Emitter.int32:NiagaraInt32=&ReturnExecIndex()) {isSetParameters=true}
  ]
  particleSpawnScript [
   &VC_PaintedVariantsOverride() {enabled=false}
   &InitializeParticle(
    Color Mode=::NewEnumerator0
    Lifetime=$Emitter.CurrentLoopDuration
    Mesh Scale Mode=::NewEnumerator3
    Mesh Uniform Scale=0.13
    UsePositionOffset=false
    Write Material Random=true
   )
   &PartitionParticles()
   &SkeletalMeshLocation(
    Bone / Socket Index=&Add_Integer(A=$Particles.PartitionParticles.Partition, B=4)
    Bone / Socket Sampling Mode=::NewEnumerator5
    Interpolate Mesh Sampling=false
    MaskSpawnGroup=false
    Mesh Sampling Type=::NewEnumerator6
    Sampled Position Offset=&AddVectorToPosition(Position=<-70 0 -25>, Vector=$User.offset)
    Skeletal Mesh=$System.SkeletalMesh
   )
   &AddVelocity(Velocity=&VC_Mask_Vectors(
     FrontEnd Value:Vector3f=<-1500 0 0>
     blsFrontEndPreview:NiagaraBool=false
     SpeedMin Value:Vector3f=<-800 0 0>
     SpeedMax Value:Vector3f=<-1 0 0>
     Speed Alpha:NiagaraFloat=$System.Speed_Alpha
    )) {enabled=false}
   &MeshRotationForce(Rotation=<0 50 0>) {module=MeshRotationForce001}
   &ApplyInitialForces()
  ]
  particleUpdateScript [
   &FN_DisableTireFX(
    Mask Index Attribute:NiagaraInt32=$Particles.PartitionParticles.Partition
    Mask Index FL=2
    Mask Index FR=3
    Mask Index RL=0
    Mask Index RR=1
   )
   &ParticleState()
   &SkeletalMeshLocation(
    Bone / Socket Index=&Add_Integer(A=$Particles.PartitionParticles.Partition, B=4)
    Bone / Socket Sampling Mode=::NewEnumerator5
    MaskSpawnGroup=false
    Mesh Sampling Type=::NewEnumerator6
    Sampled Position Offset=&AddVectorToPosition(Position=<-70 0 -25>, Vector=$User.offset)
    Skeletal Mesh=$System.SkeletalMesh
   ) {module=SkeletalMeshLocation001}
   &MeshRotationForce(Rotation=<0 1.728 0>) {enabled=false}
   &SolveForcesAndVelocity(Clamp Velocity=false, Limit Acceleration=false)
  ]
  rendererClasses [&NiagaraMeshRendererProperties, &NiagaraSpriteRendererProperties]
  renderers [
   {rendererClass=&NiagaraMeshRendererProperties, rendererIndex=0}
   {rendererClass=&NiagaraSpriteRendererProperties, rendererIndex=1}
  ]
 }
 WheelWrap_Snow001 {
  emitterUpdateScript [
   &EmitterState(Life Cycle Mode=::NewEnumerator1, Loop Behavior=::NewEnumerator1, Loop Duration Mode=::NewEnumerator1, UseLoopDelay=false)
   &SpawnBurst_Instantaneous(Spawn Count=2, Use Spawn Probability=false, UseLoopCountLimit=false)
  ]
  particleSpawnScript [
   &VC_PaintedVariantsOverride() {enabled=false}
   &InitializeParticle(
    Color Mode=::NewEnumerator0
    Mesh Scale Mode=::NewEnumerator4
    Sprite Rotation Mode=::NewEnumerator2
    Sprite Size Mode=::NewEnumerator3
    UsePositionOffset=false
    Write Material Random=true
   )
   &PartitionParticles()
   &SkeletalMeshLocation(
    Bone / Socket Index=&Add_Integer(A=$Particles.PartitionParticles.Partition, B=4)
    Bone / Socket Sampling Mode=::NewEnumerator5
    Interpolate Mesh Sampling=false
    MaskSpawnGroup=false
    Mesh Sampling Type=::NewEnumerator6
    Skeletal Mesh=$System.SkeletalMesh
   )
   &InitialMeshOrientation(Mesh Orientation Mode=::NewEnumerator7, Rotation=<0 0.3 0>, Use Rotation Vector=true)
  ]
  particleUpdateScript [
   &FN_DisableTireFX(
    Mask Index Attribute:NiagaraInt32=$Particles.PartitionParticles.Partition
    Mask Index FL=2
    Mask Index FR=3
    Mask Index RL=0
    Mask Index RR=1
   )
   &VC_WheelWrapScale(Wheel Index Attribute:NiagaraInt32=&Add_Integer(A=$Particles.PartitionParticles.Partition, B=2))
   &ParticleState(Kill Particles When Lifetime Has Elapsed=false, Loop Particles Lifetime=true)
   &UpdateMeshOrientation(Rotation Coordinate Space=::NewEnumerator2, Rotation Vector=<0 1 0>) {enabled=false}
   &SkeletalMeshLocation(
    Bone / Socket Index=&Add_Integer(A=$Particles.PartitionParticles.Partition, B=4)
    Bone / Socket Sampling Mode=::NewEnumerator5
    MaskSpawnGroup=false
    Mesh Sampling Type=::NewEnumerator6
    Skeletal Mesh=$System.SkeletalMesh
   ) {module=SkeletalMeshLocation001}
   &SolveForcesAndVelocity(Clamp Velocity=false, Limit Acceleration=false)
  ]
  rendererClasses [&NiagaraMeshRendererProperties]
  renderers [{rendererClass=&NiagaraMeshRendererProperties, rendererIndex=0}]
 }
}
summary {
 systemName=NS_DriftTrail_Daikon_MP
 emitters [
  {emitterName=Rock_Smoke, rendererClasses [&NiagaraMeshRendererProperties, &NiagaraSpriteRendererProperties]}
  {emitterName=Snowball001, rendererClasses [&NiagaraMeshRendererProperties, &NiagaraSpriteRendererProperties]}
  {emitterName=WheelWrap_Snow001, rendererClasses [&NiagaraMeshRendererProperties]}
 ]
 userVariables [
  User.bIsFrontendPreview {type=NiagaraBool, defaultValue=true}
  User.bIsPainted {type=NiagaraBool}
  User.ColorVariantID {type=NiagaraInt32}
  User.offset {type=Vector3f, description="offset from wheels(set through cosmetic template)", defaultValue {struct=Vector3f, value=<0 0 0>}}
  User.PaintedColor {type=LinearColor, description="", defaultValue {struct=LinearColor, value=<rgba 1 1 1 1>}}
  User.PaintedScalar {type=NiagaraFloat, defaultValue=1, description=""}
  User.VehicleFlags {type=NiagaraInt32, defaultValue=806289408}
  User.Wheel_Radius_Array {
   type=NiagaraDataInterfaceArrayFloat
   defaultValue {
    struct=NiagaraExt_VariableValue_DataInterface
    value {
     dataInterface=&self:NiagaraDataInterfaceArrayFloat_1
     dataInterfaceClass=&NiagaraDataInterfaceArrayFloat
    }
   }
  }
  User.Wheel_Width_Array {
   type=NiagaraDataInterfaceArrayFloat
   defaultValue {
    struct=NiagaraExt_VariableValue_DataInterface
    value {
     dataInterface=&self:NiagaraDataInterfaceArrayFloat_0
     dataInterfaceClass=&NiagaraDataInterfaceArrayFloat
    }
   }
  }
  User.WheelTouchingArray {
   type=NiagaraDataInterfaceArrayFloat
   defaultValue {
    struct=NiagaraExt_VariableValue_DataInterface
    value {
     dataInterface=&self:NiagaraDataInterfaceArrayFloat_2
     dataInterfaceClass=&NiagaraDataInterfaceArrayFloat
    }
   }
  }
 ]
}
