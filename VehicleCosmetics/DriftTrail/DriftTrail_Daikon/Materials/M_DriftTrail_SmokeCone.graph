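# M_DriftTrail_SmokeCone is an unlit additive Niagara material for the Daikon drift trail's smoke cone.
# Particle color feeds an emissive adjustment function, while particle alpha scales a main flame mask multiplied by switchable secondary and tertiary noise layers.
# Each layer fuses two channel-selectable textures by a multiply-to-add blend; tertiary textures pan, while secondary UVs shift with camera vector and optional noise offset.
# Optional, default-off UV distortion comes from two panning flow normals.
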Material /VehicleCosmetics/DriftTrail/DriftTrail_Daikon/Materials/M_DriftTrail_SmokeCone
graph {
 properties {
  bAutomaticallySetUsageInEditor=false
  bUsedWithNiagaraMeshParticles=true
  bUsedWithNiagaraRibbons=true
  bUsedWithNiagaraSprites=true
  blendMode=BLEND_Additive
  shadingModel=MSM_Unlit
  usedShadingModels=Unlit
 }
 nodes [
  MaterialExpressionVertexColor_0 = ParticleColor()
  MaterialFunctionCall(Emissive=@MaterialExpressionVertexColor_0.RGB) {materialFunction=&MF_FX_EmissiveAdjustments}
  n1 = TextureSampleParameter2D(UVs=NamedRerouteUsage()) {group=Main Texture, parameterName=Main Texture, sortPriority=0, texture=/DealershipSharedLibrary/FX/Textures/Gradient_Masks/T_VCVFX_Flame_A}
  n2 = Multiply(A=@n1.A, B=ScalarParameter() {defaultValue=0.6, group=Main Texture, parameterName=Main Texture Multiplier}) {constB=0.6}
  n3 = NamedRerouteUsage()
  n4 = ScalarParameter() {defaultValue=1, desc=Random Range -.3 to -.05 Daikon OG, group=Tertiary Textures, parameterName=Contrast Tertiary Textures Result}
  n5 = StaticSwitchParameter(True=Multiply(A=NamedRerouteUsage(), B=Power(Base=@n3, Exp=@n4)), False=NamedRerouteUsage()) {defaultValue=true, group=General Params, parameterName=Use Tertiary Textures?}
  Multiply(
   A=@MaterialExpressionVertexColor_0.A
   B=StaticSwitchParameter(True=Multiply(A=@n2, B=@n5), False=@n2) {defaultValue=true, group=General Params, parameterName=Use Secondary Textures?}
  )
  n6 = Multiply(
   A=TextureCoordinate()
   B=ScalarParameter() {defaultValue=1, desc=1.4 original daikon, group=Secondary Texture, parameterName=Tiling, sortPriority=0}
  )
  NamedRerouteDeclaration(None=@n6) {name=Tiling Secondary Textures, nodeColor=<rgba 1 0.2344 0 1>}
  n7 = TextureCoordinate() {uTiling=2, vTiling=2}
  n8 = ScalarParameter() {defaultValue=1, group=Tertiary Textures, parameterName=Tiling Tertiary Textures, sortPriority=3}
  n9 = VectorParameter() {
   group=Tertiary Textures
   parameterName=Panner Speed Tertiary Textures
   sortPriority=4
   channelNames {a=Speed T2 Y, b=Speed T2 X, g=Speed T1 Y, r=Speed T1 X}
   defaultValue=<rgba 0.2 -0.1 -0.05 -0.1>
  }
  n10 = ScalarParameter() {defaultValue=0.1, desc=Default Values Daikon .1 to .5, group=Tertiary Textures, parameterName=Speed Tertiary Textures, sortPriority=4}
  n11 = MaterialFunctionCall(Speed=@n10) {materialFunction=&TimeWithSpeedVariable}
  n12 = TextureSampleParameter2D(UVs=Panner(
    Coordinate=Multiply(A=@n7, B=@n8)
    Time=@n11.Time
    Speed=MaterialFunctionCall(X=@n9.Speed T1 X, Y=@n9.Speed T1 X) {materialFunction=&MakeFloat2}.Result
   ) {speedX=0.2, speedY=-0.1}) {
   group=Tertiary Textures
   parameterName=Tertiary Texture 1
   sortPriority=0
   texture=/DealershipSharedLibrary/FX/Textures/Noise/T_VCVFX_Energy_Alpha
  }
  n13 = VectorParameter() {group=Tertiary Textures, parameterName=RGBA Channel Selector Tertiary TEX 1, sortPriority=1, defaultValue=<rgba 1 0 0 0>}
  n14 = DotProduct(A=@n12.RGB, B=@n13.RGB)
  n15 = TextureCoordinate()
  n16 = TextureSampleParameter2D(UVs=Panner(
    Coordinate=Multiply(A=@n8, B=@n15)
    Time=@n11.Time
    Speed=MaterialFunctionCall(X=@n9.Speed T2 X, Y=@n9.Speed T2 X) {materialFunction=&MakeFloat2}.Result
   ) {speedX=-0.05, speedY=-0.1}) {
   group=Tertiary Textures
   parameterName=Tertiary Texture 2
   sortPriority=6
   texture=&T_VCVFX_Energy_Alpha
  }
  n17 = VectorParameter() {group=Tertiary Textures, parameterName=RGBA Channel Selector Tertiary TEX 2, sortPriority=8, defaultValue=<rgba 0 0 1 0>}
  n18 = DotProduct(A=@n16.RGB, B=@n17.RGB)
  n19 = Multiply(A=@n14, B=@n18)
  NamedRerouteDeclaration(None=LinearInterpolate(
    A=@n19
    B=Add(A=@n14, B=@n18)
    Alpha=ScalarParameter() {group=Tertiary Textures, parameterName="Tertiary Textures Fusion Mode 0=MUL 1=ADD", sortPriority=-1}
   )) {name=Tertiary Textures Result, nodeColor=<rgba 0 0.0859 1 1>}
  n20 = ComponentMask(None=CameraVectorWS()) {g=true, r=true}
  n21 = NamedRerouteUsage()
  n22 = Add(A=@n20, B=ComponentMask(None=ConstantBiasScale(None=@n21) {bias=-0.5, scale=2}) {g=true, r=true})
  n23 = TextureSampleParameter2D(UVs=Add(
    A=NamedRerouteUsage()
    B=StaticSwitchParameter(True=@n22, False=@n20) {defaultValue=true, group=General Params, parameterName=Use Tertiary Textures?}
   )) {
   group=Secondary Texture
   parameterName=Secondary Texture 1
   sortPriority=3
   texture=/DealershipSharedLibrary/FX/Textures/Noise/T_VCVFX_NoiseGeneric01Pack_RGB
  }
  n24 = VectorParameter() {group=Secondary Texture, parameterName=RGBA Channel Selector Secondary TEX 1, sortPriority=4, defaultValue=<rgba 0 0 1 0>}
  n25 = DotProduct(A=@n23.RGB, B=@n24.RGB)
  n26 = TextureSampleParameter2D(UVs=NamedRerouteUsage()) {
   group=Secondary Texture
   parameterName=Secondary Texture 2
   sortPriority=5
   texture=&T_VCVFX_NoiseGeneric01Pack_RGB
  }
  n27 = VectorParameter() {group=Tertiary Textures, parameterName=RGBA Channel Selector Secondary TEX 2, sortPriority=6, defaultValue=<rgba 0 0 1 0>}
  n28 = DotProduct(A=@n26.RGB, B=@n27.RGB) {desc=Default B}
  MaterialExpressionConstantClamp_0 = Saturate(None=LinearInterpolate(
    A=Multiply(A=@n25, B=@n28)
    B=Add(A=@n25, B=@n28)
    Alpha=ScalarParameter() {group=Secondary Texture, parameterName="Secondary Textures Fusion Mode 0=MUL 1=ADD", sortPriority=1}
   ))
  n29 = Multiply(
   A=@MaterialExpressionConstantClamp_0
   B=ScalarParameter() {defaultValue=512, group=Secondary Texture, parameterName=Secondary Textures Multiplier, sortPriority=2}
  )
  NamedRerouteDeclaration(None=@n29) {name=Secondary Textures, nodeColor=<rgba 1 0 0.3281 1>}
  n30 = TextureCoordinate()
  n31 = TextureCoordinate()
  n32 = ScalarParameter() {defaultValue=1, group=Distortion, parameterName=Tiling Distortion}
  n33 = MaterialFunctionCall(Speed=ScalarParameter() {group=Distortion, parameterName=Speed Distortion}) {materialFunction=&TimeWithSpeedVariable}
  n34 = VectorParameter() {
   group=Distortion
   parameterName=Panner Speed Distortion
   channelNames {a=Texture 2 Y, b=Texture 2 X, g=Texture 1 Y, r=Texture 1 X}
   defaultValue=<rgba 0.2 0.1 -0.12 -0.2>
  }
  n35 = TextureSampleParameter2D(UVs=Panner(
    Coordinate=Multiply(A=@n31, B=@n32)
    Time=@n33.Time
    Speed=MaterialFunctionCall(X=@n34.Texture 1 X, Y=@n34.Texture 1 X) {materialFunction=&MakeFloat2}.Result
   ) {speedX=0.2, speedY=0.1}) {
   group=Distortion
   parameterName=Distortion Texture 1
   samplerType=SAMPLERTYPE_Normal
   texture=/DealershipSharedLibrary/FX/Textures/Normal/T_VCVFX_FlowN03_N
  }
  n36 = TextureCoordinate() {uTiling=0.5, vTiling=0.5}
  n37 = TextureSampleParameter2D(UVs=Panner(
    Coordinate=Multiply(A=@n32, B=@n36)
    Time=@n33.Time
    Speed=MaterialFunctionCall(X=@n34.Texture 2 X, Y=@n34.Texture 2 X) {materialFunction=&MakeFloat2}.Result
   ) {speedX=-0.12, speedY=-0.2}) {
   group=Distortion
   parameterName=Distortion Texture 2
   samplerType=SAMPLERTYPE_Normal
   texture=&T_VCVFX_FlowN03_N
  }
  n38 = ComponentMask(None=Add(A=@n35.RGB, B=@n37.RGB)) {g=true, r=true}
  n39 = ScalarParameter() {defaultValue=0.1, group=Distortion, parameterName=Distortion Amount}
  n40 = StaticSwitchParameter(True=Add(A=@n30, B=Multiply(A=@n38, B=@n39)), False=@n30) {group=General Params, parameterName=Use Distortion?}
  NamedRerouteDeclaration(None=@n40) {name=Distortion, nodeColor=<rgba 1 0.8672 0 1>}
  Reroute(None=@n29)
  Reroute(None=@n40)
  Reroute(None=@n19)
  Reroute(None=@n6)
 ]
}
